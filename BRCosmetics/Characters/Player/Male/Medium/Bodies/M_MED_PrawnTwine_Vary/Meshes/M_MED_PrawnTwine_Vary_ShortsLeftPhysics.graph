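# Physics asset that simulates the left leg of the shorts on the PrawnTwine_Vary medium male player body.
# Front, back, inner and outer two-link capsule chains hang from base bones joined through a general shorts bone to the thigh.
# Chain joints lock twist and keep swing limits tight, while base-to-general and general-to-thigh constraints leave rotation free.
# A larger thigh capsule gives the leg a body alongside the cloth bones.
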PhysicsAsset /BRCosmetics/Characters/Player/Male/Medium/Bodies/M_MED_PrawnTwine_Vary/Meshes/M_MED_PrawnTwine_Vary_ShortsLeftPhysics
bodies {
 base_shorts_bk_l [{shapeType=Capsule, length=0.505, radius=0.505, shapeName=base_shorts_bk_l_capsule, rotation=<rot 0 -90 0>}]
 base_shorts_fr_l [{shapeType=Capsule, length=0.505, radius=0.505, shapeName=base_shorts_fr_l_capsule, rotation=<rot 0 -90 0>}]
 base_shorts_general_l [{shapeType=Capsule, length=0.505, radius=0.505, shapeName=base_shorts_general_l_capsule, rotation=<rot 0 -90 0>}]
 base_shorts_in_l [{shapeType=Capsule, length=0.505, radius=0.505, shapeName=base_shorts_in_l_capsule, rotation=<rot 0 -90 0>}]
 base_shorts_out_l [{shapeType=Capsule, length=0.505, radius=0.505, shapeName=base_shorts_out_l_capsule, rotation=<rot 0 -90 0>}]
 dyn_shorts_bk_01_l [{shapeType=Capsule, length=10, radius=4, shapeName=dyn_shorts_bk_01_l_capsule, center=<5.6477 0 1.9745>, rotation=<rot 0 0 90>}]
 dyn_shorts_bk_02_l [{shapeType=Capsule, length=10, radius=4, shapeName=dyn_shorts_bk_02_l_capsule, center=<5.6477 0 1.4357>, rotation=<rot 0 0 90>}]
 dyn_shorts_fr_01_l [{shapeType=Capsule, length=11.6518, radius=4, shapeName=dyn_shorts_fr_01_l_capsule, center=<5.6477 0 4.1053>, rotation=<rot 0 0 90>}]
 dyn_shorts_fr_02_l [{shapeType=Capsule, length=11.6518, radius=4, shapeName=dyn_shorts_fr_02_l_capsule, center=<5.6477 0 4.0067>, rotation=<rot 0 0 90>}]
 dyn_shorts_in_01_l [{shapeType=Capsule, length=10, radius=4, shapeName=dyn_shorts_in_01_l_capsule, center=<5.6477 0 3.3748>, rotation=<rot 0 0 90>}]
 dyn_shorts_in_02_l [{shapeType=Capsule, length=10, radius=4, shapeName=dyn_shorts_in_02_l_capsule, center=<5.6477 0 3.1652>, rotation=<rot 0 0 90>}]
 dyn_shorts_out_01_l [{shapeType=Capsule, length=11.5672, radius=4, shapeName=dyn_shorts_out_01_l_capsule, center=<5.6477 0 3.6112>, rotation=<rot 0 0 90>}]
 dyn_shorts_out_02_l [{shapeType=Capsule, length=11.5672, radius=4, shapeName=dyn_shorts_out_02_l_capsule, center=<5.6477 0 4.9993>, rotation=<rot 0 0 90>}]
 thigh_l [{shapeType=Capsule, length=26.51, radius=6.0442, shapeName=thigh_l_capsule, center=<-25.9889 1.8656 -1.3147>, rotation=<rot 90 0 0>}]
}
constraints [
 {
  bone1Name=dyn_shorts_bk_01_l
  bone2Name=base_shorts_bk_l
  swing1LimitDegrees=10
  swing2LimitDegrees=3
  twistLimitDegrees=45
  twistMotion=Locked
 }
 {
  bone1Name=dyn_shorts_bk_02_l
  bone2Name=dyn_shorts_bk_01_l
  swing1LimitDegrees=10
  swing2LimitDegrees=3
  twistLimitDegrees=45
  twistMotion=Locked
 }
 {
  bone1Name=dyn_shorts_fr_01_l
  bone2Name=base_shorts_fr_l
  swing1LimitDegrees=5
  swing2LimitDegrees=10
  twistLimitDegrees=45
  twistMotion=Locked
 }
 {
  bone1Name=dyn_shorts_fr_02_l
  bone2Name=dyn_shorts_fr_01_l
  swing1LimitDegrees=5
  swing2LimitDegrees=10
  twistLimitDegrees=45
  twistMotion=Locked
 }
 {
  bone1Name=dyn_shorts_in_01_l
  bone2Name=base_shorts_in_l
  swing1LimitDegrees=3
  swing2LimitDegrees=5
  twistLimitDegrees=45
  twistMotion=Locked
 }
 {
  bone1Name=dyn_shorts_in_02_l
  bone2Name=dyn_shorts_in_01_l
  swing1LimitDegrees=3
  swing2LimitDegrees=5
  twistLimitDegrees=45
  twistMotion=Locked
 }
 {
  bone1Name=dyn_shorts_out_01_l
  bone2Name=base_shorts_out_l
  swing1LimitDegrees=3
  swing2LimitDegrees=6
  twistLimitDegrees=45
  twistMotion=Locked
 }
 {
  bone1Name=dyn_shorts_out_02_l
  bone2Name=dyn_shorts_out_01_l
  swing1LimitDegrees=3
  swing2LimitDegrees=6
  twistLimitDegrees=45
  twistMotion=Locked
 }
 {
  bone1Name=base_shorts_fr_l
  bone2Name=base_shorts_general_l
  swing1LimitDegrees=45
  swing1Motion=Free
  swing2LimitDegrees=45
  swing2Motion=Free
  twistLimitDegrees=45
  twistMotion=Free
 }
 {
  bone1Name=base_shorts_out_l
  bone2Name=base_shorts_general_l
  swing1LimitDegrees=45
  swing1Motion=Free
  swing2LimitDegrees=45
  swing2Motion=Free
  twistLimitDegrees=45
  twistMotion=Free
 }
 {
  bone1Name=base_shorts_bk_l
  bone2Name=base_shorts_general_l
  swing1LimitDegrees=45
  swing1Motion=Free
  swing2LimitDegrees=45
  swing2Motion=Free
  twistLimitDegrees=45
  twistMotion=Free
 }
 {
  bone1Name=base_shorts_in_l
  bone2Name=base_shorts_general_l
  swing1LimitDegrees=45
  swing1Motion=Free
  swing2LimitDegrees=45
  swing2Motion=Free
  twistLimitDegrees=45
  twistMotion=Free
 }
 {
  bone1Name=base_shorts_general_l
  bone2Name=thigh_l
  swing1LimitDegrees=45
  swing1Motion=Free
  swing2LimitDegrees=45
  swing2Motion=Free
  twistLimitDegrees=45
  twistMotion=Free
 }
]
