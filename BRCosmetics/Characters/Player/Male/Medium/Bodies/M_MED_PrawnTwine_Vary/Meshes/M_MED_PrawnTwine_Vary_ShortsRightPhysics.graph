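# Physics asset for the right leg of the shorts on the PrawnTwine_Vary medium male player body cosmetic.
# A thigh capsule connects through a general shorts base to back, front, inner and outer base bodies, all with free angular constraints.
# Each base carries a two-link chain of dyn capsules with locked twist and tight swing limits.
# The front chain also locks one swing axis, restricting it to a single bend direction.
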PhysicsAsset /BRCosmetics/Characters/Player/Male/Medium/Bodies/M_MED_PrawnTwine_Vary/Meshes/M_MED_PrawnTwine_Vary_ShortsRightPhysics
bodies {
 base_shorts_bk_r [{shapeType=Capsule, length=0.505, radius=0.505, shapeName=base_shorts_bk_r_capsule, rotation=<rot 0 -90 180>}]
 base_shorts_fr_r [{shapeType=Capsule, length=0.505, radius=0.505, shapeName=base_shorts_fr_r_capsule, rotation=<rot 0 -90 180>}]
 base_shorts_general_r [{shapeType=Capsule, length=0.505, radius=0.505, shapeName=base_shorts_general_r_capsule, rotation=<rot 0 -90 0>}]
 base_shorts_in_r [{shapeType=Capsule, length=0.505, radius=0.505, shapeName=base_shorts_in_r_capsule, rotation=<rot 0 -90 180>}]
 base_shorts_out_r [{shapeType=Capsule, length=0.505, radius=0.505, shapeName=base_shorts_out_r_capsule, rotation=<rot 0 -90 180>}]
 dyn_shorts_bk_01_r [{shapeType=Capsule, length=10, radius=4, shapeName=dyn_shorts_bk_01_r_capsule, center=<-5.6477 0 -1.9745>, rotation=<rot 0 0 -90>}]
 dyn_shorts_bk_02_r [{shapeType=Capsule, length=10, radius=4, shapeName=dyn_shorts_bk_02_r_capsule, center=<-5.6477 0 -1.4357>, rotation=<rot 0 0 -90>}]
 dyn_shorts_fr_01_r [{shapeType=Capsule, length=11.6518, radius=4, shapeName=dyn_shorts_fr_01_r_capsule, center=<-5.6477 0 -4.1053>, rotation=<rot 0 0 -90>}]
 dyn_shorts_fr_02_r [{shapeType=Capsule, length=11.6518, radius=4, shapeName=dyn_shorts_fr_02_r_capsule, center=<-5.6477 0 -3.5442>, rotation=<rot 0 0 -90>}]
 dyn_shorts_in_01_r [{shapeType=Capsule, length=10, radius=4, shapeName=dyn_shorts_in_01_r_capsule, center=<-5.6477 0 -3.7171>, rotation=<rot 0 0 -90>}]
 dyn_shorts_in_02_r [{shapeType=Capsule, length=10, radius=4, shapeName=dyn_shorts_in_02_r_capsule, center=<-5.6477 0 -3.1456>, rotation=<rot 0 0 -90>}]
 dyn_shorts_out_01_r [{shapeType=Capsule, length=11.5672, radius=4, shapeName=dyn_shorts_out_01_r_capsule, center=<-5.6477 0 -4.4376>, rotation=<rot 0 0 -90>}]
 dyn_shorts_out_02_r [{shapeType=Capsule, length=11.5672, radius=4, shapeName=dyn_shorts_out_02_r_capsule, center=<-5.6477 0 -6.0964>, rotation=<rot 0 0 -90>}]
 thigh_r [{shapeType=Capsule, length=26.51, radius=6.0442, shapeName=thigh_r_capsule, center=<25.9889 -1.8656 1.3147>, rotation=<rot 90 180 0>}]
}
constraints [
 {
  bone1Name=base_shorts_bk_r
  bone2Name=base_shorts_general_r
  swing1LimitDegrees=45
  swing1Motion=Free
  swing2LimitDegrees=45
  swing2Motion=Free
  twistLimitDegrees=45
  twistMotion=Free
 }
 {
  bone1Name=dyn_shorts_bk_01_r
  bone2Name=base_shorts_bk_r
  swing1LimitDegrees=10
  swing2LimitDegrees=6
  twistLimitDegrees=45
  twistMotion=Locked
 }
 {
  bone1Name=dyn_shorts_bk_02_r
  bone2Name=dyn_shorts_bk_01_r
  swing1LimitDegrees=10
  swing2LimitDegrees=6
  twistLimitDegrees=45
  twistMotion=Locked
 }
 {
  bone1Name=base_shorts_fr_r
  bone2Name=base_shorts_general_r
  swing1LimitDegrees=45
  swing1Motion=Free
  swing2LimitDegrees=45
  swing2Motion=Free
  twistLimitDegrees=45
  twistMotion=Free
 }
 {
  bone1Name=dyn_shorts_fr_01_r
  bone2Name=base_shorts_fr_r
  swing1LimitDegrees=5
  swing1Motion=Locked
  swing2LimitDegrees=10
  twistLimitDegrees=45
  twistMotion=Locked
 }
 {
  bone1Name=dyn_shorts_fr_02_r
  bone2Name=dyn_shorts_fr_01_r
  swing1LimitDegrees=5
  swing1Motion=Locked
  swing2LimitDegrees=10
  twistLimitDegrees=45
  twistMotion=Locked
 }
 {
  bone1Name=base_shorts_in_r
  bone2Name=base_shorts_general_r
  swing1LimitDegrees=45
  swing1Motion=Free
  swing2LimitDegrees=45
  swing2Motion=Free
  twistLimitDegrees=45
  twistMotion=Free
 }
 {
  bone1Name=dyn_shorts_in_01_r
  bone2Name=base_shorts_in_r
  swing1LimitDegrees=3
  swing2LimitDegrees=5
  twistLimitDegrees=45
  twistMotion=Locked
 }
 {
  bone1Name=dyn_shorts_in_02_r
  bone2Name=dyn_shorts_in_01_r
  swing1LimitDegrees=3
  swing2LimitDegrees=5
  twistLimitDegrees=45
  twistMotion=Locked
 }
 {
  bone1Name=base_shorts_out_r
  bone2Name=base_shorts_general_r
  swing1LimitDegrees=45
  swing1Motion=Free
  swing2LimitDegrees=45
  swing2Motion=Free
  twistLimitDegrees=45
  twistMotion=Free
 }
 {
  bone1Name=dyn_shorts_out_01_r
  bone2Name=base_shorts_out_r
  swing1LimitDegrees=3
  swing2LimitDegrees=6
  twistLimitDegrees=45
  twistMotion=Locked
 }
 {
  bone1Name=dyn_shorts_out_02_r
  bone2Name=dyn_shorts_out_01_r
  swing1LimitDegrees=3
  swing2LimitDegrees=6
  twistLimitDegrees=45
  twistMotion=Locked
 }
 {
  bone1Name=base_shorts_general_r
  bone2Name=thigh_r
  swing1LimitDegrees=45
  swing1Motion=Free
  swing2LimitDegrees=45
  swing2Motion=Free
  twistLimitDegrees=45
  twistMotion=Free
 }
]
