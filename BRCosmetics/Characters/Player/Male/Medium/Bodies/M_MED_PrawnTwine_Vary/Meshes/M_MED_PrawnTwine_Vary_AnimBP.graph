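# Post-process style AnimBlueprint for the PrawnTwine_Vary male body cosmetic, layering secondary motion onto the pose copied from the attached parent mesh.
# A Control Rig receives crouch and pistol values, then RigidBody shorts simulations and AnimDynamics tail, sleeve and bagginess bones feed a cached pose.
# Additive bone offsets then adjust shorts, tail and feet.
# The event graph updates gravity and physics settings, parachute shirt-dynamics alphas, and a two-emote tail flag.
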AnimBlueprint /BRCosmetics/Characters/Player/Male/Medium/Bodies/M_MED_PrawnTwine_Vary/Meshes/M_MED_PrawnTwine_Vary_AnimBP
target_skeleton=&/M_MED_PrawnTwine_Vary_Skeleton
graphs [
 AnimGraph {nodes [
   n1 = "Transform (Modify) Bone - Bone: base_shorts_fr_r"!(
    ComponentPose=Local To Component(LocalPose=Use cached pose 'DynamicsCache'())
    bAlphaBoolEnabled:bool=True
    Translation="2.898,-0.139,1.635"
    Rotation="-27.192,12.088,-5.589"
    Scale="1,1,1"
   ) -> (Pose) {
    alphaInputType=Bool
    rotationMode=BMM_Additive
    rotationSpace=BCS_BoneSpace
    translationMode=BMM_Additive
    translationSpace=BCS_BoneSpace
    alphaBoolBlend {blendInTime=0.2, blendOutTime=0.2}
    boneToModify {boneName=base_shorts_fr_r}
    rotation=<rot -27.192 12.088 -5.589>
    translation=<2.8979 -0.139 1.635>
   }
   n2 = "Transform (Modify) Bone - Bone: base_shorts_out_r"!(ComponentPose=@n1, bAlphaBoolEnabled:bool=True, Translation="-1.380,-0.207,1.499", Rotation="-3.688,7.997,-0.602", Scale="1,1,1") -> (Pose) {
    alphaInputType=Bool
    rotationMode=BMM_Additive
    rotationSpace=BCS_BoneSpace
    translationMode=BMM_Additive
    translationSpace=BCS_BoneSpace
    alphaBoolBlend {blendInTime=0.2, blendOutTime=0.2}
    boneToModify {boneName=base_shorts_out_r}
    rotation=<rot -3.688 7.997 -0.602>
    translation=<-1.38 -0.207 1.499>
   }
   n3 = "Transform (Modify) Bone - Bone: base_shorts_fr_r"(ComponentPose=@n2, bAlphaBoolEnabled=True, Translation="-1.539,-4.653,2.107", Rotation="14.543,-16.102,-4.889", Scale="1,1,1") {
    alphaInputType=Bool
    rotationMode=BMM_Additive
    rotationSpace=BCS_BoneSpace
    translationMode=BMM_Additive
    translationSpace=BCS_BoneSpace
    alphaBoolBlend {blendInTime=0.2, blendOutTime=0.2}
    boneToModify {boneName=base_shorts_fr_r}
    rotation=<rot 14.543 -16.102 -4.889>
    translation=<-1.539 -4.653 2.107>
   }
   n4 = "Transform (Modify) Bone - Bone: base_tail"!(
    ComponentPose=@n3
    bAlphaBoolEnabled:bool=True
    Translation="-3.695,-0.000,-0.919"
    Rotation="13.954,-0.000,0.000"
    Scale="0.601,0.601,0.601"
   ) -> (Pose) {
    alphaInputType=Bool
    rotationMode=BMM_Additive
    rotationSpace=BCS_BoneSpace
    scaleMode=BMM_Additive
    scaleSpace=BCS_BoneSpace
    translationMode=BMM_Additive
    translationSpace=BCS_BoneSpace
    alphaBoolBlend {blendInTime=0.5, blendOption=CubicInOut, blendOutTime=0.2}
    alphaScaleBiasClamp {bInterpResult=true, interpSpeedIncreasing=5, outRange {max=0.5}}
    boneToModify {boneName=base_tail}
    rotation=<rot 13.954 0 0>
    scale=<0.601 0.601 0.601>
    translation {x=-3.695, z=-0.919}
   }
   n5 = "Transform (Modify) Bone - Bone: foot_l"!(ComponentPose=@n4, bAlphaBoolEnabled:bool=True, Translation="-0.000,-7.210,-0.000", Rotation="0,0,0", Scale="1,0.5,1") -> (Pose) {
    alphaInputType=Bool
    rotationMode=BMM_Additive
    rotationSpace=BCS_BoneSpace
    scaleMode=BMM_Additive
    scaleSpace=BCS_BoneSpace
    translationMode=BMM_Additive
    translationSpace=BCS_BoneSpace
    alphaBoolBlend {blendInTime=0.5, blendOption=ExpIn, blendOutTime=0.2}
    alphaScaleBiasClamp {bInterpResult=true, interpSpeedIncreasing=3}
    boneToModify {boneName=foot_l}
    scale {y=0.5}
    translation {y=-7.21}
   }
   n6 = "Transform (Modify) Bone - Bone: foot_r"!(ComponentPose=@n5, bAlphaBoolEnabled:bool=True, Translation="-0.000,7.129,-0.649", Rotation="0,0,0", Scale="1,0.5,1") -> (Pose) {
    alphaInputType=Bool
    rotationMode=BMM_Additive
    rotationSpace=BCS_BoneSpace
    scaleMode=BMM_Additive
    scaleSpace=BCS_BoneSpace
    translationMode=BMM_Additive
    translationSpace=BCS_BoneSpace
    alphaBoolBlend {blendInTime=0.5, blendOption=ExpIn, blendOutTime=0.2}
    alphaScaleBiasClamp {bInterpResult=true, interpSpeedIncreasing=3}
    boneToModify {boneName=foot_r}
    scale {y=0.5}
    translation {y=7.129, z=-0.649}
   }
   Output Pose(Result=Component To Local(ComponentPose=@n6)) {layerGroup=None}
   n7 = Control Rig!(
    Source=Copy Pose From Mesh() {bUseAttachedParent=true}
    Alpha
    OnlyCrouching:double=Get OnlyCrouching() -> (OnlyCrouching:double)
    Pistol:double=Get Pistol() -> (Pistol:double)
   ) -> (Pose) {controlRigAssetReference {blueprintRigClass=&/M_MED_PrawnTwine_Vary_CtrlRig.M_MED_PrawnTwine_Vary_CtrlRig_C}}
   Save cached pose 'ControlRigCache'(Pose=@n7) {cacheName=ControlRigCache}
   n8 = RigidBody!(
    ComponentPose=Local To Component(LocalPose=Use cached pose 'ControlRigCache'())
    Alpha
    OverrideWorldGravity:Vector="0,0,-980"
    ExternalForce
    SimSpaceSettings:Sim Space Settings Structure=(WorldAlpha=0,VelocityScaleZ=1,DampingAlpha=1,MaxLinearVelocity=10000,MaxAngularVelocity=10000,MaxLinearAcceleration=10000,MaxAngularAcceleration=10000,ExternalLinearDragV=(X=0,Y=0,Z=0),ExternalLinearVelocity=(X=0,Y=0,Z=0),ExternalAngularVelocity=(X=0,Y=0,Z=0))
   ) -> (Pose) {
    bOverrideWorldGravity=true
    overridePhysicsAsset=&/M_MED_PrawnTwine_Vary_ShortsLeftPhysics
    simulationSpace=BaseBoneSpace
    baseBoneRef {boneName=thigh_l}
    componentAppliedLinearAccClamp=<0 0 0>
    overrideWorldGravity {z=-980}
   }
   n9 = RigidBody(
    ComponentPose=@n8
    OverrideWorldGravity="0,0,-980"
    SimSpaceSettings=(WorldAlpha=0,VelocityScaleZ=1,DampingAlpha=1,MaxLinearVelocity=10000,MaxAngularVelocity=10000,MaxLinearAcceleration=10000,MaxAngularAcceleration=10000,ExternalLinearDragV=(X=0,Y=0,Z=0),ExternalLinearVelocity=(X=0,Y=0,Z=0),ExternalAngularVelocity=(X=0,Y=0,Z=0))
   ) {
    bOverrideWorldGravity=true
    overridePhysicsAsset=&/M_MED_PrawnTwine_Vary_ShortsRightPhysics
    simulationSpace=BaseBoneSpace
    baseBoneRef {boneName=thigh_r}
    componentAppliedLinearAccClamp=<0 0 0>
    overrideWorldGravity {z=-980}
   }
   n10 = Anim Dynamics - Bone: dyn_tail!(ComponentPose=@n9, Alpha=0.8, GravityOverride:Vector="0,0,-980", ExternalForce="0,0,0") -> (Pose) {
    angularDampingOverride=0.9999
    angularSpringConstant=15000
    bAngularSpring=true
    bOverrideAngularDamping=true
    gravityScale=0.35
    simulationSpace=RootRelative
    boundBone {boneName=dyn_tail}
    componentAppliedLinearAccClamp=<2000 2000 2000>
    componentLinearAccScale=<1 1 1>
    componentLinearVelScale=<1 1 1>
    physicsBodyDefinitions [{
      boundBone {boneName=dyn_tail}
      boxExtents=<20 30 30>
      constraintSetup {angularLimitsMax=<0 14 14>, angularLimitsMin=<0 -14 -14>, angularTarget=<1 0 0>}
      localJointOffset=<2 0 0>
     }]
   }
   n11 = Anim Dynamics - Bone: dyn_tail_1!(ComponentPose=@n10, Alpha=1, GravityOverride:Vector="0,0,-980", ExternalForce="0,0,0") -> (Pose) {
    angularDampingOverride=0.9999
    angularSpringConstant=15000
    bAngularSpring=true
    bOverrideAngularDamping=true
    gravityScale=0.35
    simulationSpace=RootRelative
    boundBone {boneName=dyn_tail_1}
    componentAppliedLinearAccClamp=<2000 2000 2000>
    componentLinearAccScale=<1 1 1>
    componentLinearVelScale=<1 1 1>
    gravityOverride {z=-980}
    physicsBodyDefinitions [{
      boundBone {boneName=dyn_tail_1}
      boxExtents=<15 25 25>
      constraintSetup {angularLimitsMax=<0 20 20>, angularLimitsMin=<0 -20 -20>, angularTarget=<1 0 0>}
      localJointOffset=<4 0 0>
     }]
   }
   n12 = Anim Dynamics - Bone: dyn_tail_a_l!(ComponentPose=@n11, Alpha=1, GravityOverride:Vector="0,0,-980", ExternalForce="0,0,0") -> (Pose) {
    angularDampingOverride=0.9999
    angularSpringConstant=25000
    bAngularSpring=true
    bOverrideAngularDamping=true
    gravityScale=0.35
    simulationSpace=RootRelative
    boundBone {boneName=dyn_tail_a_l}
    componentAppliedLinearAccClamp=<2000 2000 2000>
    componentLinearAccScale=<1 1 1>
    componentLinearVelScale=<1 1 1>
    gravityOverride {z=-980}
    physicsBodyDefinitions [{
      boundBone {boneName=dyn_tail_a_l}
      boxExtents=<15 25 25>
      constraintSetup {angularLimitsMax=<0 10 10>, angularLimitsMin=<0 -10 -10>, angularTarget=<1 0 0>}
      localJointOffset=<2 0 0>
     }]
   }
   n13 = Anim Dynamics - Bone: dyn_tail_a_r!(ComponentPose=@n12, Alpha=1, GravityOverride:Vector="0,0,-980", ExternalForce="0,0,0") -> (Pose) {
    angularDampingOverride=0.9999
    angularSpringConstant=25000
    bAngularSpring=true
    bOverrideAngularDamping=true
    gravityScale=0.35
    simulationSpace=RootRelative
    boundBone {boneName=dyn_tail_a_r}
    componentAppliedLinearAccClamp=<2000 2000 2000>
    componentLinearAccScale=<1 1 1>
    componentLinearVelScale=<1 1 1>
    gravityOverride {z=-980}
    physicsBodyDefinitions [{
      boundBone {boneName=dyn_tail_a_r}
      boxExtents=<15 25 25>
      constraintSetup {angularLimitsMax=<0 10 10>, angularLimitsMin=<0 -10 -10>, angularTarget=<1 0 0>}
      localJointOffset=<2 0 0>
     }]
   }
   n14 = Anim Dynamics - Bone: dyn_uniformArm_sd_01_l!(ComponentPose=@n13, Alpha=0.7, GravityOverride:Vector="0,0,-980", ExternalForce="0,0,0") -> (Pose) {
    angularDampingOverride=0.999
    angularSpringConstant=500
    bAngularSpring=true
    bLinearSpring=true
    bOverrideAngularDamping=true
    bOverrideLinearDamping=true
    bUseGravityOverride=true
    gravityScale=0.35
    linearDampingOverride=0.9
    linearSpringConstant=0.9
    simulationSpace=RootRelative
    alphaScaleBiasClamp {bInterpResult=true}
    boundBone {boneName=dyn_uniformArm_sd_01_l}
    componentAppliedLinearAccClamp=<2000 2000 2000>
    componentLinearAccScale=<1 1 1>
    componentLinearVelScale=<1 1 1>
    gravityOverride {z=-980}
    physicsBodyDefinitions [{
      boundBone {boneName=dyn_uniformArm_sd_01_l}
      boxExtents=<10 10 10>
      constraintSetup {angularLimitsMax=<50 50 50>, angularLimitsMin=<-50 -10 -50>, angularTarget=<1 0 0>, linearAxesMax=<0 0 2>, linearAxesMin=<0 0 -2>}
      localJointOffset=<4 0 0>
     }]
    simSpaceSettings {maxAngularAcceleration=100, maxAngularVelocity=50, simSpaceAngularAlpha=1}
   }
   n15 = Anim Dynamics - Bone: dyn_uniformArm_sd_01_r!(ComponentPose=@n14, Alpha=0.7, GravityOverride:Vector="0,0,-980", ExternalForce="0,0,0") -> (Pose) {
    angularDampingOverride=0.999
    angularSpringConstant=500
    bAngularSpring=true
    bLinearSpring=true
    bOverrideAngularDamping=true
    bOverrideLinearDamping=true
    bUseGravityOverride=true
    gravityScale=0.35
    linearDampingOverride=0.9
    linearSpringConstant=200
    simulationSpace=RootRelative
    alphaScaleBiasClamp {bInterpResult=true}
    boundBone {boneName=dyn_uniformArm_sd_01_r}
    componentAppliedLinearAccClamp=<2000 2000 2000>
    componentLinearAccScale=<1 1 1>
    componentLinearVelScale=<1 1 1>
    gravityOverride {z=-980}
    physicsBodyDefinitions [{
      boundBone {boneName=dyn_uniformArm_sd_01_r}
      boxExtents=<10 10 10>
      constraintSetup {angularLimitsMax=<50 50 50>, angularLimitsMin=<-50 -10 -50>, angularTarget=<1 0 0>, linearAxesMax=<0 0 2>, linearAxesMin=<0 0 -2>}
      localJointOffset=<-4 0 0>
     }]
    simSpaceSettings {maxAngularAcceleration=100, maxAngularVelocity=50, simSpaceAngularAlpha=1}
   }
   n16 = Anim Dynamics - Bone: dyn_uniformArm_sd_01_l(ComponentPose=@n15, Alpha=0.7, GravityOverride="0,0,-980", ExternalForce="0,0,0") {
    angularDampingOverride=0.999
    angularSpringConstant=500
    bAngularSpring=true
    bLinearSpring=true
    bOverrideAngularDamping=true
    bOverrideLinearDamping=true
    bUseGravityOverride=true
    gravityScale=0.35
    linearDampingOverride=0.9
    linearSpringConstant=0.9
    simulationSpace=RootRelative
    alphaScaleBiasClamp {bInterpResult=true}
    boundBone {boneName=dyn_uniformArm_sd_01_l}
    componentAppliedLinearAccClamp=<2000 2000 2000>
    componentLinearAccScale=<1 1 1>
    componentLinearVelScale=<1 1 1>
    gravityOverride {z=-980}
    physicsBodyDefinitions [{
      boundBone {boneName=dyn_uniformArm_sd_01_l}
      boxExtents=<10 10 10>
      constraintSetup {angularLimitsMax=<50 0 50>, angularLimitsMin=<-50 0 -50>, angularTarget=<1 0 0>, linearAxesMax=<0 0 2>, linearAxesMin=<0 0 -2>}
      localJointOffset=<4 0 0>
     }]
    simSpaceSettings {maxAngularAcceleration=100, maxAngularVelocity=50, simSpaceAngularAlpha=1}
   }
   n17 = Anim Dynamics - Bone: dyn_uniformArm_sd_01_r(ComponentPose=@n16, Alpha=0.7, GravityOverride="0,0,-980", ExternalForce="0,0,0") {
    angularDampingOverride=0.999
    angularSpringConstant=500
    bAngularSpring=true
    bLinearSpring=true
    bOverrideAngularDamping=true
    bOverrideLinearDamping=true
    bUseGravityOverride=true
    gravityScale=0.35
    linearDampingOverride=0.9
    linearSpringConstant=200
    simulationSpace=RootRelative
    alphaScaleBiasClamp {bInterpResult=true}
    boundBone {boneName=dyn_uniformArm_sd_01_r}
    componentAppliedLinearAccClamp=<2000 2000 2000>
    componentLinearAccScale=<1 1 1>
    componentLinearVelScale=<1 1 1>
    gravityOverride {z=-980}
    physicsBodyDefinitions [{
      boundBone {boneName=dyn_uniformArm_sd_01_r}
      boxExtents=<10 10 10>
      constraintSetup {angularLimitsMax=<50 0 50>, angularLimitsMin=<-50 0 -50>, angularTarget=<1 0 0>, linearAxesMax=<0 0 2>, linearAxesMin=<0 0 -2>}
      localJointOffset=<-4 0 0>
     }]
    simSpaceSettings {maxAngularAcceleration=100, maxAngularVelocity=50, simSpaceAngularAlpha=1}
   }
   n18 = Anim Dynamics - Bone: dyn_uniform_bagginess_sd_01_l(ComponentPose=@n17, Alpha=1, ExternalForce="0,0,0") {
    angularDampingOverride=0.9999
    angularSpringConstant=8000
    bAngularSpring=true
    bOverrideAngularDamping=true
    bUseGravityOverride=true
    gravityScale=0.35
    linearSpringConstant=0.5
    simulationSpace=BoneRelative
    alphaBoolBlend {blendOption=Linear}
    boundBone {boneName=dyn_uniform_bagginess_sd_01_l}
    componentAppliedLinearAccClamp=<3000 3000 3000>
    componentLinearAccScale=<0.4 0.4 0.4>
    componentLinearVelScale=<0.4 0.4 0.4>
    gravityOverride {z=-500}
    physicsBodyDefinitions [{
      boundBone {boneName=dyn_uniform_bagginess_sd_01_l}
      boxExtents=<20 25 25>
      constraintSetup {angularLimitsMax=<15 40 20>, angularLimitsMin=<-15 -40 -20>, angularTarget=<1 0 0>, linearAxesMax=<0 1 5>}
      localJointOffset=<-5 0 0>
     }]
    relativeSpaceBone {boneName=root}
   }
   n19 = Anim Dynamics - Bone: dyn_uniform_bagginess_sd_01_r(ComponentPose=@n18, Alpha=1, ExternalForce="0,0,0") {
    angularDampingOverride=0.9999
    angularSpringConstant=8000
    bAngularSpring=true
    bOverrideAngularDamping=true
    bUseGravityOverride=true
    gravityScale=0.35
    linearSpringConstant=0.5
    simulationSpace=BoneRelative
    alphaBoolBlend {blendOption=Linear}
    boundBone {boneName=dyn_uniform_bagginess_sd_01_r}
    componentAppliedLinearAccClamp=<3000 3000 3000>
    componentLinearAccScale=<0.4 0.4 0.4>
    componentLinearVelScale=<0.4 0.4 0.4>
    gravityOverride {z=-500}
    physicsBodyDefinitions [{
      boundBone {boneName=dyn_uniform_bagginess_sd_01_r}
      boxExtents=<20 25 25>
      constraintSetup {angularLimitsMax=<15 40 20>, angularLimitsMin=<-15 -40 -20>, angularTarget=<1 0 0>, linearAxesMax=<0 1 0>, linearAxesMin=<0 0 -5>}
      localJointOffset=<5 0 0>
     }]
    relativeSpaceBone {boneName=root}
   }
   n20 = Anim Dynamics - Bone: dyn_uniform_bagginess_bk_01(ComponentPose=@n19, Alpha=1, ExternalForce="0,0,0") {
    angularDampingOverride=0.9999
    angularSpringConstant=8000
    bAngularSpring=true
    bOverrideAngularDamping=true
    bUseGravityOverride=true
    gravityScale=0.35
    linearDampingOverride=0.6
    linearSpringConstant=0.5
    simulationSpace=BoneRelative
    alphaBoolBlend {blendOption=Linear}
    boundBone {boneName=dyn_uniform_bagginess_bk_01}
    componentAppliedLinearAccClamp=<3000 3000 3000>
    componentLinearAccScale=<0.4 0.4 0.4>
    componentLinearVelScale=<0.4 0.4 0.4>
    gravityOverride {z=-500}
    physicsBodyDefinitions [{
      boundBone {boneName=dyn_uniform_bagginess_bk_01}
      boxExtents=<20 25 25>
      constraintSetup {angularLimitsMax=<15 40 20>, angularLimitsMin=<-15 -40 -20>, angularTarget=<1 0 0>, linearAxesMax=<0 1 5>}
      localJointOffset=<-5 0 0>
     }]
    relativeSpaceBone {boneName=root}
   }
   Save cached pose 'DynamicsCache'(Pose=Component To Local(ComponentPose=@n20)) {cacheName=DynamicsCache}
  ]}
 EventGraph {nodes [
   Event BlueprintUpdateAnimation()
   > n1 = Sequence!(execute) -> (then_0, then_1, then_2:Exec, then_3:Exec)
   > n2 = "Gravity Override State Function (Parameter Asset) (Nativized)"(
    CharacterPartAnimInstance=Self-Reference()
    GameTimeInSeconds=GetGameTimeInSeconds()
    Parameters=&/M_MED_PrawnTwine_Vary_Parameters
   )
   > Set Gravity Override(Gravity Override=@n2)
   > Set RBAN Settings Out(RBAN Settings Out=@n2.OutRigidBodyAnimNodeParameters)
   n3 = NOT Boolean(A=Get bIsSliding())
   n4 = Get bIsCrouching()
   Set OnlyCrouching(execute=@n1.then_1, OnlyCrouching:double="To Float (Boolean)"(InBool=AND Boolean(A=@n4, B=@n3))) -> (Output_Get:double)
   n5 = GetAnimInstance(self=GetOwnerSkeletalMeshComponent())
   Set AdjustTailToEmotes(
    execute=@n1.then_3
    AdjustTailToEmotes:bool=OR Boolean(
     A=Montage_IsPlaying(self=@n5, Montage=/BRCosmetics/Animation/Game/MainPlayer/Emotes/Cauldron/Emote_Cauldron_CMM_M)
     B=Montage_IsPlaying(self=@n5, Montage=/BRCosmetics/Animation/Game/MainPlayer/Emotes/Scrub_A_Dub/Emote_ScrubaDub_CMM_M)
    )
   ) -> (Output_Get:bool)
   n6 = Branch(execute=@n1.then_2, Condition=Get bIsParachuteOpened())
   > Set Alpha_ShirtDynDefault(Alpha_ShirtDynDefault:double=0) -> (Output_Get:double)
   > Set Alpha_ShirtDynParachute(Alpha_ShirtDynParachute:double=0.7) -> (Output_Get:double)
   Set Alpha_ShirtDynDefault(execute=@n6.else, Alpha_ShirtDynDefault=0.7)
   > Set Alpha_ShirtDynParachute(Alpha_ShirtDynParachute=0)
   Event OnWeaponChanged()
   > n7 = Cast To FortPlayerAnimInstance(Object=GetAnimInstance(self=GetOwnerSkeletalMeshComponent()))
   > Set Pistol(Pistol:double=0) -> (Output_Get:double)
   > n8 = Switch on EFortWeaponCoreAnimation!(execute, Selection:EFortWeaponCoreAnimation Enum=Get WeaponCoreAnim(self=@n7)) -> (
    Melee:Exec
    Pistol:Exec
    Shotgun:Exec
    PaperBlueprint:Exec
    Rifle:Exec
    MeleeOneHand:Exec
    MachinePistol:Exec
    RocketLauncher:Exec
    GrenadeLauncher:Exec
    GoingCommando:Exec
    AssaultRifle:Exec
    TacticalShotgun:Exec
    SniperRifle:Exec
    TrapPlacement:Exec
    ShoulderLauncher:Exec
    AbilityDecoTool:Exec
    Crossbow:Exec
    C4:Exec
    RemoteControl:Exec
    DualWield:Exec
    AR_BullPup:Exec
    AR_ForwardGrip:Exec
    MedPackPaddles:Exec
    SMG_P90:Exec
    AR_DrumGun:Exec
    Consumable_Small:Exec
    Consumable_Large:Exec
    Balloon:Exec
    MountedTurret:Exec
    CreativeTool:Exec
    ExplosiveBow:Exec
    AshtonIndigo:Exec
    AshtonChicago:Exec
    MeleeDualWield:Exec
    Unarmed:Exec
   )
   Set Pistol(execute=@n8.Pistol, Pistol=1)
  ]}
]
variables [
 AdjustTailToEmotes {type=boolean}
 Alpha_ShirtDynDefault {type=number}
 Alpha_ShirtDynParachute {type=number}
 Gravity Override {type=Vector}
 OnlyCrouching {type=number}
 Pistol {type=number}
 RBAN Settings Out {type=FortRigidBodyAnimNodeParameters}
]
